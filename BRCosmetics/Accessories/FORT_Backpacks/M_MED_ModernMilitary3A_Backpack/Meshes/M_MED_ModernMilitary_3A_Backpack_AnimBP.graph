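# Animation Blueprint for the ModernMilitary 3A backpack cosmetic, adding procedural swing and secondary motion to the incoming pose.
# Each update maps left upper-arm yaw relative to the clavicle to swing_rotate_alpha, zeroed during skydiving or when custom animations can play.
# That alpha drives an additive backpack_root rotation, then AnimDynamics springs move the radio, ammo and cloth bones.
# The top-cloth dynamics and test sequence player are not wired into the output.
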AnimBlueprint /BRCosmetics/Accessories/FORT_Backpacks/M_MED_ModernMilitary3A_Backpack/Meshes/M_MED_ModernMilitary_3A_Backpack_AnimBP
target_skeleton=/Game/Accessories/FORT_Backpacks/M_MED_ModernMilitary3A_Backpack/Meshes/M_MED_ModernMilitary_3A_Backpack_Skeleton
graphs [
 AnimGraph {nodes [
   n1 = Anim Dynamics - Bone: dyn_cloth_top_l!(ComponentPose, ExternalForce="0,0,0") -> (Pose) {
    alpha=0.2
    bGravityOverrideInSimSpace=true
    bLinearSpring=true
    bOverrideLinearDamping=true
    gravityScale=0.2
    lODThreshold=1
    linearDampingOverride=0.99
    linearSpringConstant=320
    simulationSpace=Actor
    alphaBoolBlend {blendOption=Linear}
    boundBone {boneName=dyn_cloth_top_l}
    physicsBodyDefinitions [{
      sphereCollisionRadius=0
      boundBone {boneName=dyn_cloth_top_l}
      boxExtents=<30 25 15>
      constraintSetup {
       twistAxis=AxisZ
       angularLimitsMax=<0 0 0>
       angularLimitsMin=<0 0 0>
       angularTarget=<0 0 0>
       linearAxesMax=<10 0 10>
       linearAxesMin=<-10 0 -10>
      }
      localJointOffset=<0 0 0>
     }]
   }
   Anim Dynamics - Bone: dyn_cloth_top_r!(ComponentPose=@n1, ExternalForce="0,0,0") -> (Pose) {
    alpha=0.2
    bGravityOverrideInSimSpace=true
    bLinearSpring=true
    bOverrideLinearDamping=true
    gravityScale=0.2
    lODThreshold=1
    linearDampingOverride=0.99
    linearSpringConstant=320
    simulationSpace=Actor
    alphaBoolBlend {blendOption=Linear}
    boundBone {boneName=dyn_cloth_top_r}
    physicsBodyDefinitions [{
      sphereCollisionRadius=0
      boundBone {boneName=dyn_cloth_top_r}
      boxExtents=<30 25 15>
      constraintSetup {
       twistAxis=AxisZ
       angularLimitsMax=<0 0 0>
       angularLimitsMin=<0 0 0>
       angularTarget=<0 0 0>
       linearAxesMax=<10 0 10>
       linearAxesMin=<-10 0 -10>
      }
      localJointOffset=<0 0 0>
     }]
   }
   AnimGraphNode_SubInput_0 = Input Pose()
   n2 = "Transform (Modify) Bone - Bone: backpack_root"(
    ComponentPose=Local To Component(LocalPose=@AnimGraphNode_SubInput_0)
    Alpha=Get swing_rotate_alpha() -> (swing_rotate_alpha:double)
    Translation="0,0,0"
    Rotation="50,0,20"
    Scale="1,1,1"
   ) {
    rotationMode=BMM_Additive
    rotationSpace=BCS_BoneSpace
    alphaBoolBlend {blendOption=Linear}
    alphaScaleBiasClamp {bInterpResult=true, interpSpeedDecreasing=5}
    boneToModify {boneName=backpack_root}
    rotation {pitch=50, roll=20}
   }
   n3 = Anim Dynamics - Bone: dyn_radio(ComponentPose=@n2, Alpha=0.35, ExternalForce="0,0,0") {
    angularDampingOverride=0.99
    angularSpringConstant=3500
    bAngularSpring=true
    bGravityOverrideInSimSpace=true
    bOverrideAngularDamping=true
    numSolverIterationsPostUpdate=4
    numSolverIterationsPreUpdate=16
    alphaBoolBlend {blendOption=Linear}
    boundBone {boneName=dyn_radio}
    physicsBodyDefinitions [{
      sphereCollisionRadius=0
      boundBone {boneName=dyn_radio}
      boxExtents=<20 20 10>
      constraintSetup {angularTargetAxis=AxisY, twistAxis=AxisZ, angularLimitsMax=<40 20 5>, angularLimitsMin=<-15 -20 -5>, angularTarget=<0 1 0>}
      localJointOffset=<0 0 -3>
     }]
   }
   n4 = Anim Dynamics - Bone: dyn_ammo(ComponentPose=@n3, Alpha=0.35, ExternalForce="0,0,0") {
    angularDampingOverride=0.99
    angularSpringConstant=3500
    bAngularSpring=true
    bGravityOverrideInSimSpace=true
    bOverrideAngularDamping=true
    numSolverIterationsPostUpdate=4
    numSolverIterationsPreUpdate=16
    alphaBoolBlend {blendOption=Linear}
    boundBone {boneName=dyn_ammo}
    physicsBodyDefinitions [{
      sphereCollisionRadius=0
      boundBone {boneName=dyn_ammo}
      boxExtents=<20 20 10>
      constraintSetup {angularTargetAxis=AxisY, twistAxis=AxisZ, angularLimitsMax=<40 20 5>, angularLimitsMin=<-15 -20 -5>, angularTarget=<0 1 0>}
      localJointOffset=<0 0 -3>
     }]
   }
   n5 = "Anim Dynamics - Chain: dyn_cloth_l_01 -> dyn_cloth_l_02"!(ComponentPose=@n4, ExternalForce="0,0,0") -> (Pose) {
    alpha=0.4
    angularDampingOverride=0.99
    angularSpringConstant=500
    bAngularSpring=true
    bChain=true
    bGravityOverrideInSimSpace=true
    bOverrideAngularDamping=true
    numSolverIterationsPostUpdate=4
    numSolverIterationsPreUpdate=16
    simulationSpace=Actor
    alphaBoolBlend {blendOption=Linear}
    boundBone {boneName=dyn_cloth_l_01}
    chainEnd {boneName=dyn_cloth_l_02}
    componentAppliedLinearAccClamp=<5000 5000 5000>
    componentLinearAccScale=<0.5 0.5 2>
    componentLinearVelScale=<0.5 0.5 2>
    physicsBodyDefinitions [{
      sphereCollisionRadius=0
      boundBone {boneName=dyn_cloth_l_01}
      boxExtents=<10 20 20>
      constraintSetup {angularLimitsMax=<5 60 60>, angularLimitsMin=<-5 -60 -5>, angularTarget=<1 0 0>}
      localJointOffset=<-2 0 0>
     }]
    relativeSpaceBone {boneName=backpack_root}
   }
   n6 = "Anim Dynamics - Chain: dyn_cloth_m_01 -> dyn_cloth_m_02"!(ComponentPose=@n5, ExternalForce="0,0,0") -> (Pose) {
    alpha=0.4
    angularDampingOverride=0.99
    angularSpringConstant=500
    bAngularSpring=true
    bChain=true
    bGravityOverrideInSimSpace=true
    bOverrideAngularDamping=true
    numSolverIterationsPostUpdate=4
    numSolverIterationsPreUpdate=16
    simulationSpace=Actor
    alphaBoolBlend {blendOption=Linear}
    boundBone {boneName=dyn_cloth_m_01}
    chainEnd {boneName=dyn_cloth_m_02}
    componentAppliedLinearAccClamp=<5000 5000 5000>
    componentLinearAccScale=<0.5 0.5 2>
    componentLinearVelScale=<0.5 0.5 2>
    physicsBodyDefinitions [{
      sphereCollisionRadius=0
      boundBone {boneName=dyn_cloth_m_01}
      boxExtents=<12 25 19>
      constraintSetup {angularLimitsMax=<5 60 60>, angularLimitsMin=<-5 -60 -5>, angularTarget=<1 0 0>}
      localJointOffset=<-2 0 0>
     }]
    relativeSpaceBone {boneName=backpack_root}
   }
   n7 = "Anim Dynamics - Chain: dyn_cloth_r_01 -> dyn_cloth_r_02"!(ComponentPose=@n6, ExternalForce="0,0,0") -> (Pose) {
    alpha=0.4
    angularDampingOverride=0.99
    angularSpringConstant=500
    bAngularSpring=true
    bChain=true
    bGravityOverrideInSimSpace=true
    bOverrideAngularDamping=true
    numSolverIterationsPostUpdate=4
    numSolverIterationsPreUpdate=16
    simulationSpace=Actor
    alphaBoolBlend {blendOption=Linear}
    boundBone {boneName=dyn_cloth_r_01}
    chainEnd {boneName=dyn_cloth_r_02}
    componentAppliedLinearAccClamp=<5000 5000 5000>
    componentLinearAccScale=<0.5 0.5 2>
    componentLinearVelScale=<0.5 0.5 2>
    physicsBodyDefinitions [{
      sphereCollisionRadius=0
      boundBone {boneName=dyn_cloth_r_01}
      boxExtents=<18 30 28>
      constraintSetup {angularLimitsMax=<5 60 60>, angularLimitsMin=<-5 -60 -5>, angularTarget=<1 0 0>}
      localJointOffset=<-2 0 0>
     }]
    relativeSpaceBone {boneName=backpack_root}
   }
   n8 = Anim Dynamics - Bone: dyn_cloth_side!(ComponentPose=@n7, ExternalForce="0,0,0") -> (Pose) {
    alpha=0.4
    angularDampingOverride=0.99
    angularSpringConstant=300
    bAngularSpring=true
    bGravityOverrideInSimSpace=true
    bOverrideAngularDamping=true
    numSolverIterationsPostUpdate=4
    numSolverIterationsPreUpdate=16
    simulationSpace=Actor
    alphaBoolBlend {blendOption=Linear}
    boundBone {boneName=dyn_cloth_side}
    physicsBodyDefinitions [{
      sphereCollisionRadius=0
      boundBone {boneName=dyn_cloth_side}
      boxExtents=<18 30 28>
      constraintSetup {angularLimitsMax=<5 60 10>, angularLimitsMin=<-5 -60 -10>, angularTarget=<1 0 0>}
      localJointOffset=<-4 0 0>
     }]
   }
   Output Pose(Result=Component To Local(ComponentPose=@n8)) {layerGroup=None}
   Sequence Player 'Backpack_PhysicsTest_Animation'() {sequence=/Game/Accessories/FORT_Backpacks/M_MED_ModernMilitary3A_Backpack/Meshes/Backpack_PhysicsTest_Animation}
  ]}
 EventGraph {nodes [
   Event BlueprintUpdateAnimation()
   n1 = BreakTransform(InTransform=GetDeltaTransformFromRefPose(self=GetOwnerSkeletalMeshComponent(), BoneName=upperarm_twist_01_l, BaseName=clavicle_l))
   > Set swing_rotate_alpha(swing_rotate_alpha:double=MapRangeClamped(Value=BreakRotator(InRot=@n1.Rotation).Yaw, InRangeA=10, InRangeB=-40)) -> (Output_Get:double)
   > Branch(Condition=OR Boolean(A=Get bIsSkyDiving(), B=Get bCanPlayCustomAnimations() -> (bCanPlayCustomAnimations:bool)))
   > Set swing_rotate_alpha(swing_rotate_alpha=0)
  ]}
]
variables [swing_rotate_alpha {type=number, default=0.2}]
